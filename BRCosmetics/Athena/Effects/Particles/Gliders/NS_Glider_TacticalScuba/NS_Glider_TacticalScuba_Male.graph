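# Niagara effect for the male Tactical Scuba glider cosmetic, adding bubbles, wisp trails and a colored light to the glider's skeletal mesh.
# Sprite emitters spawn bubbles at the props and wings; a turning emitter uses RightDot to scale spawn rate and bias which wing emits.
# Two ribbons form wisps, and one bone-following particle renders a sprite and light tinted by LightColor.
# Front-end modifiers swap several velocities, gravities, rates and offsets for menu display.
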NiagaraSystem /BRCosmetics/Athena/Effects/Particles/Gliders/NS_Glider_TacticalScuba/NS_Glider_TacticalScuba_Male
dependencies {
 usedDataInterfaces [&NiagaraDataInterfaceCurve, &NiagaraDataInterfaceVectorCurve]
 usedDynamicInputs []
 usedModules [
  &self:SystemSpawnScript.SystemScriptSource.SystemScriptGraph.NiagaraNodeAssignment_0.SetVariables_#1
  &SystemState
 ]
}
emitters {
 Bubbles_Wings {
  emitterSpawnScript [&FN_PreserveSkydiveUserVariables() {enabled=false, module=PreserveUserVariables}]
  emitterUpdateScript [
   &EmitterState(Inactive Response=::NewEnumerator1, Life Cycle Mode=::NewEnumerator1, Loop Behavior=::NewEnumerator1, Loop Duration=3600, UseLoopDelay=false)
   &SpawnRate(SpawnRate=200, Use Spawn Probability=false)
  ]
  particleSpawnScript [
   /SetVariables(Particles.BoneIndex:NiagaraInt32=&UniformRangedInt(Minimum=0, Maximum=1)) {isSetParameters=true, module=SetVariables_#2}
   /Game/Effects/Niagara/Modules/Groups/FN_SetSpawnGroup_v1/FN_SetSpawnGroup(Emitter Bone Count=2) {enabled=false}
   /Niagara/Modules/Spawn/Initialization/InitializeParticle(
    Lifetime=&FloatFromCurve(
     FloatCurve=DataInterface() {propertyValues {Curve {keys [<at 0 0 RCIM_Cubic>, {arriveTangent=3.0737, interpMode=RCIM_Cubic, leaveTangent=3.0737, tangentMode=RCTM_User, time=1, value=1}]}}}
     CurveIndex=&UniformRangedFloat(Minimum=0, Maximum=1)
    )
    Module.Write Color=true
    Module.Write Lifetime=true
    Module.Write Mass=true
    Module.Write Material Random=true
    Module.Write Position=false
    Module.Write Scale=false
    Module.Write SpriteRotation=false
    Module.Write SpriteSize=false
    Position=&SimulationPosition()
   )
   /SetVariables(Particles.Scale=&VectorFromFloat(Value=100)) {enabled=false, isSetParameters=true, module=SetVariables_#3}
   &SampleSkeletalMeshSkeleton(
    Mesh:NiagaraDataInterfaceSkeletalMesh=$User.SkeletalMesh
    Spawn Mode:ENiagaraSkeletalSpawning=::NewEnumerator1
    Specific Bone Index:NiagaraInt32=$Particles.BoneIndex
    Specific Socket Index:NiagaraInt32=$Particles.BoneIndex
   )
   /Niagara/Modules/Spawn/Location/SkeletalMeshSkeletonLocation(Module.Inherit Velocity=false, Module.Use Sampled Bone as Ribbon ID=true)
   &CylinderLocation(
    Cylinder Height=500
    Cylinder Radius=25
    MaskSpawnGroup=false
    Module.Enable Lathe Profile=false
    Module.Override Randomness=false
    Module.Surface Only=false
    Orientation Axis:ENiagaraOrientationAxis=::YAxis
   ) {enabled=false}
   &BoxLocation(
    Box Offset=<0 0 -5>
    Box Size=<200 40 0>
    MaskSpawnGroup=false
    Module.Override Seed=false
    Normalized Box Rotation Angle:NiagaraFloat=&FloatFromCurve(
     FloatCurve=DataInterface() {propertyValues {Curve {keys [<at 0 -0.1>,<at 1 0.1>]}}}
     CurveIndex=&MakeFloatFromInt(Int=$Particles.BoneIndex)
     Scale Curve=0.6
    )
    Rotate Box=true
    Surface Only=false
   )
   /Niagara/Modules/Update/SubUV/SubUVAnimation(SubUV Animation Mode=::NewEnumerator1)
   /SetVariables(Particles.SpriteSize=&Vector2DFromFloat(Value=&UniformRangedFloat(Minimum=2, Maximum=5))) {isSetParameters=true, module=SetVariables_#4}
   /SetVariables(Particles.SpriteSize=<30 30>) {enabled=false, isSetParameters=true, module=SetVariables_#5}
   &GravityForce(Gravity=&FN_FrontEndVector3Modifier(Gameplay=<0 0 -400>, Front End=<0 0 75>))
   &ApplyInitialForces()
   &AddVelocity(Velocity=&FN_FrontEndVector3Modifier(
     Gameplay=&UniformRangedVector(Minimum=<-10 150 -10>, Maximum=<10 300 10>)
     Front End=&UniformRangedVector(Minimum=<-10 -150 -10>, Maximum=<10 -300 10>)
    ))
  ]
  particleUpdateScript [
   &ParticleState()
   &SolveForcesAndVelocity(Clamp Velocity=false, Limit Acceleration=false)
   /SetVariables(Particles.Velocity=/Niagara/DynamicInputs/ScaleByCurve/ScaleVectorByCurve(
     InitialValue:Vector3f=&UniformRangedVector(Minimum=<-3 -3 0>, Maximum=<3 3 0>)
     ScaleCurve:NiagaraDataInterfaceVectorCurve=DataInterface() {propertyValues {
       XCurve {keys [<at 0 50 RCIM_Cubic>, {arriveTangent=-241.6722, interpMode=RCIM_Cubic, leaveTangent=-241.672, tangentMode=RCTM_User, time=0.5, value=20}]}
       YCurve {keys [<at 0 50 RCIM_Cubic>, {arriveTangent=-247.443, interpMode=RCIM_Cubic, leaveTangent=-247.448, tangentMode=RCTM_User, time=0.4991, value=20}]}
       ZCurve {keys [<at 0 0>,<at 1 0>]}
      }}
     CurveIndex:NiagaraFloat=$Particles.NormalizedAge
    )) {enabled=false, isSetParameters=true, module=SetVariables_#6}
   /SetVariables() {enabled=false, isSetParameters=true, module=SetVariables_#7}
   &ScaleSpriteSize(Scale Factor=&Vector2DFromFloat(Value=&FloatFromCurve(FloatCurve=DataInterface() {propertyValues {Curve {keys [<at 0 0 RCIM_Cubic>,<at 0.05 1 RCIM_Cubic>,<at 1 0>]}}})))
  ]
  rendererClasses [&NiagaraSpriteRendererProperties]
  renderers [
   {rendererClass=&NiagaraSpriteRendererProperties}
   {rendererClass=&NiagaraSpriteRendererProperties, rendererIndex=1}
  ]
 }
 Bubbles_Wings_Turning {
  emitterSpawnScript [&FN_PreserveSkydiveUserVariables() {enabled=false, module=PreserveUserVariables}]
  emitterUpdateScript [
   &EmitterState(Inactive Response=::NewEnumerator1, Life Cycle Mode=::NewEnumerator1, Loop Behavior=::NewEnumerator1, Loop Duration=3600, UseLoopDelay=false)
   &SpawnRate(
    SpawnRate=&FloatFromCurve(FloatCurve=DataInterface() {propertyValues {Curve {keys [<at -1 1>,<at 0 0>,<at 1 1>]}}}, CurveIndex=$User.RightDot, Scale Curve=500)
    Use Spawn Probability=false
   )
   &SpawnRate(SpawnRate=5000, Use Spawn Probability=false) {enabled=false, module=SpawnRate001}
  ]
  particleSpawnScript [
   /SetVariables(Particles.BoneIndex:NiagaraInt32=&MakeIntFromFloat(Float=&Lerp_Float(
      A=&UniformAOrBFloat(A=0)
      B=&FloatFromCurve(
       FloatCurve=DataInterface() {propertyValues {Curve {keys [
           {arriveTangent=-0.0301, time=-1, value=1}
           {arriveTangent=-1.7191, interpMode=RCIM_Cubic, leaveTangent=-1.7191, tangentMode=RCTM_User, time=1, value=0}
          ]}}}
       CurveIndex=$User.RightDot
      )
      Alpha=&FloatFromCurve(
       FloatCurve=DataInterface() {propertyValues {Curve {keys [<at 0 0 RCIM_Constant>,<at 0.25 1>]}}}
       CurveIndex=&UniformRangedFloat(Minimum=0, Maximum=1)
      )
     ))) {isSetParameters=true, module=SetVariables_#2}
   &FN_SetSpawnGroup(Emitter Bone Count=2) {enabled=false}
   &InitializeParticle(
    Lifetime=&FloatFromCurve(
     FloatCurve=DataInterface() {propertyValues {Curve {keys [<at 0 0 RCIM_Cubic>, {arriveTangent=3.0737, interpMode=RCIM_Cubic, leaveTangent=3.0737, tangentMode=RCTM_User, time=1, value=1}]}}}
     CurveIndex=&UniformRangedFloat(Minimum=0, Maximum=1)
    )
    Module.Write Color=true
    Module.Write Lifetime=true
    Module.Write Mass=true
    Module.Write Material Random=true
    Module.Write Position=false
    Module.Write Scale=false
    Module.Write SpriteRotation=false
    Module.Write SpriteSize=false
    Position=&SimulationPosition()
   )
   /SetVariables(Particles.Scale=&VectorFromFloat(Value=100)) {enabled=false, isSetParameters=true, module=SetVariables_#3}
   &SampleSkeletalMeshSkeleton(
    Mesh:NiagaraDataInterfaceSkeletalMesh=$User.SkeletalMesh
    Spawn Mode:ENiagaraSkeletalSpawning=::NewEnumerator1
    Specific Bone Index:NiagaraInt32=$Particles.BoneIndex
    Specific Socket Index:NiagaraInt32=$Particles.BoneIndex
   )
   &SkeletalMeshSkeletonLocation(Module.Inherit Velocity=false, Module.Use Sampled Bone as Ribbon ID=true)
   &CylinderLocation(
    Cylinder Height=500
    Cylinder Radius=25
    MaskSpawnGroup=false
    Module.Enable Lathe Profile=false
    Module.Override Randomness=false
    Module.Surface Only=false
    Orientation Axis:ENiagaraOrientationAxis=::YAxis
   ) {enabled=false}
   &BoxLocation(
    Box Size=<200 10 0>
    MaskSpawnGroup=false
    Module.Override Seed=false
    Normalized Box Rotation Angle:NiagaraFloat=&FloatFromCurve(
     FloatCurve=DataInterface() {propertyValues {Curve {keys [<at 0 -0.1>,<at 1 0.1>]}}}
     CurveIndex=&MakeFloatFromInt(Int=$Particles.BoneIndex)
     Scale Curve=0.6
    )
    Rotate Box=true
    Surface Only=false
   )
   &SubUVAnimation(SubUV Animation Mode=::NewEnumerator1)
   /SetVariables(Particles.SpriteSize=&Vector2DFromFloat(Value=&UniformRangedFloat(Minimum=3, Maximum=10))) {isSetParameters=true, module=SetVariables_#4}
   /SetVariables(Particles.SpriteSize=<30 30>) {enabled=false, isSetParameters=true, module=SetVariables_#5}
   &GravityForce(Gravity=<0 0 -400>)
   &ApplyInitialForces()
   &AddVelocity(Velocity=&UniformRangedVector(Minimum=<-10 150 -10>, Maximum=<10 300 10>))
  ]
  particleUpdateScript [
   &ParticleState()
   &SolveForcesAndVelocity(Clamp Velocity=false, Limit Acceleration=false)
   /SetVariables(Particles.Velocity=&ScaleVectorByCurve(
     InitialValue:Vector3f=&UniformRangedVector(Minimum=<-3 -3 0>, Maximum=<3 3 0>)
     ScaleCurve:NiagaraDataInterfaceVectorCurve=DataInterface() {propertyValues {
       XCurve {keys [<at 0 50 RCIM_Cubic>, {arriveTangent=-241.6722, interpMode=RCIM_Cubic, leaveTangent=-241.672, tangentMode=RCTM_User, time=0.5, value=20}]}
       YCurve {keys [<at 0 50 RCIM_Cubic>, {arriveTangent=-247.443, interpMode=RCIM_Cubic, leaveTangent=-247.448, tangentMode=RCTM_User, time=0.4991, value=20}]}
       ZCurve {keys [<at 0 0>,<at 1 0>]}
      }}
     CurveIndex:NiagaraFloat=$Particles.NormalizedAge
    )) {enabled=false, isSetParameters=true, module=SetVariables_#6}
   /SetVariables() {enabled=false, isSetParameters=true, module=SetVariables_#7}
   &ScaleSpriteSize(Scale Factor=&Vector2DFromFloat(Value=&FloatFromCurve(FloatCurve=DataInterface() {propertyValues {Curve {keys [<at 0 0 RCIM_Cubic>,<at 0.05 1 RCIM_Cubic>,<at 1 0>]}}})))
  ]
  rendererClasses [&NiagaraSpriteRendererProperties]
  renderers [
   {rendererClass=&NiagaraSpriteRendererProperties}
   {rendererClass=&NiagaraSpriteRendererProperties, rendererIndex=1}
  ]
 }
 Particle_Light {
  emitterUpdateScript [
   &EmitterScalability(
    Module.bApplySpawnCountScale=false
    Module.bApplySpawnCountScaleByDistanceFraction=false
    Module.bUseMaxDistance=false
    Module.bUseMinDistance=false
   ) {enabled=false}
   &EmitterLifeCycle(NextLoopDuration=1) {enabled=false}
   &EmitterState()
   &SpawnRate(Use Spawn Probability=false) {enabled=false}
   &SpawnBurst_Instantaneous(Spawn Count=1, Spawn Probability=HlslExpression() {hlslExpression="Emitter.LoopCount == 0 ? 1.0 : 0"}, Use Spawn Probability=true)
  ]
  particleSpawnScript [
   /SetVariables(Particles.BoneIndex=4) {isSetParameters=true}
   &InitializeParticle(
    Color=$User.LightColor
    Module.Write Color=true
    Module.Write Lifetime=true
    Module.Write Mass=true
    Module.Write Material Random=true
    Module.Write Position=true
    Module.Write Scale=false
    Module.Write SpriteRotation=false
    Module.Write SpriteSize=true
    Position=&SimulationPosition()
    Sprite Size=<25 25>
   )
   &Color(Color=$User.LightColor, Scale Color=&MakeVectorFromLinearColorRGB(Color=<rgba 1 1 1 1>)) {module=Color001}
   &Color(Color=$User.LightColor, Scale Color=&MakeVectorFromLinearColorRGB(Color=<rgba 1000 100 825 1>)) {enabled=false}
  ]
  particleUpdateScript [
   &ParticleState(
    Kill Particles When Lifetime Has Elapsed=false
    Let Infinitely Lived Particles Die When Emitter Deactivates=true
    Loop Particles Lifetime=true
   )
   &ScaleColor(Scale RGB=&Multiply_VectorByFloat(Vector=<1 1 1>, Float=25), ScaleA=true, ScaleRGB=true)
   &ScaleColor(
    Scale RGB=&MakeVectorFromLinearColorRGB(Color=&MultiplyLinearColor(
      Linear Color=$User.LightColor
      Scale Factor=&MakeVector4_FromFloat(Float=&FloatFromCurve(
        FloatCurve=DataInterface() {propertyValues {Curve {keys [<at 0 1>,<at 0.0375 0.2415>,<at 0.0648 0.709>,<at 0.1086 0.5073>,<at 0.1359 0.9382>,<at 0.1578 0.324>,<at 0.1824 0.8098>,<at 0.2097 0.6632>,<at 0.2562 0.9107>,<at 0.289 0.4523>,<at 0.33 0.3423>,<at 0.3491 0.9565>,<at 0.4394 0.654>,<at 0.4503 0.1498>,<at 0.4776 0.4065>,<at 0.587 0.5348>,<at 0.6034 0.324>,<at 0.6608 0.9473>,<at 0.6936 0.5532>,<at 0.7292 0.929>,<at 0.792 0.7133>,<at 0.8166 0.1223>,<at 0.8412 0.7182>,<at 0.8713 0.0859>,<at 0.8986 0.8557>,<at 0.9123 0.379>,<at 0.9615 0.9565>,<at 1 0.5>]}}}
        Scale Curve=25
       ))
     ))
    ScaleA=true
    ScaleRGB=true
   ) {enabled=false, module=ScaleColor001}
   &ScaleSpriteSize(Scale Factor=<5 5>)
   &SampleSkeletalMeshSkeleton(
    Mesh:NiagaraDataInterfaceSkeletalMesh=$User.SkeletalMesh
    Specific Bone Index:NiagaraInt32=$Particles.BoneIndex
    Specific Socket Index:NiagaraInt32=$Particles.BoneIndex
   )
   &SkeletalMeshSkeletonLocation(Module.Inherit Velocity=false, Module.Use Sampled Bone as Ribbon ID=true)
   &InheritVelocity(Module.Limit Speed=true, Velocity Scale=<0.1 0.1 0>) {enabled=false}
   &SolveForcesAndVelocity(Clamp Velocity=false, Limit Acceleration=false) {enabled=false}
   /Game/Effects/Niagara/Modules/Position/FN_OffsetPositionByOrientationAxis(Offset Axis:Vector3f=&FN_FrontEndVector3Modifier(Gameplay=<0 50 10>, Front End=<0 35 5>))
  ]
  rendererClasses [&NiagaraLightRendererProperties, &NiagaraSpriteRendererProperties]
  renderers [
   {rendererClass=&NiagaraLightRendererProperties, rendererIndex=0}
   {rendererClass=&NiagaraSpriteRendererProperties, rendererIndex=1}
  ]
 }
 Prop_Bubbles {
  emitterSpawnScript [&FN_PreserveSkydiveUserVariables() {enabled=false, module=PreserveUserVariables}]
  emitterUpdateScript [
   &EmitterState(Inactive Response=::NewEnumerator1, Life Cycle Mode=::NewEnumerator1, Loop Behavior=::NewEnumerator1, Loop Duration=3600, UseLoopDelay=false)
   &SpawnRate(SpawnRate=&FN_FrontEndFloatModifier(Gameplay=1000, Front End=600), Use Spawn Probability=false)
  ]
  particleSpawnScript [
   &FN_SetSpawnGroup(Emitter Bone Count=2)
   &InitializeParticle(
    Lifetime=&FloatFromCurve(
     FloatCurve=DataInterface() {propertyValues {Curve {keys [<at 0 0 RCIM_Cubic>, {arriveTangent=3.0737, interpMode=RCIM_Cubic, leaveTangent=3.0737, tangentMode=RCTM_User, time=1, value=1}]}}}
     CurveIndex=&UniformRangedFloat(Minimum=0, Maximum=1)
    )
    Module.Write Color=true
    Module.Write Lifetime=true
    Module.Write Mass=true
    Module.Write Material Random=true
    Module.Write Position=false
    Module.Write Scale=false
    Module.Write SpriteRotation=false
    Module.Write SpriteSize=false
    Position=&SimulationPosition()
   )
   /SetVariables(Particles.Scale=&VectorFromFloat(Value=100)) {enabled=false, isSetParameters=true, module=SetVariables_#3}
   &SampleSkeletalMeshSkeleton(Mesh:NiagaraDataInterfaceSkeletalMesh=$User.SkeletalMesh, Specific Bone Index:NiagaraInt32=$Particles.SpawnGroup)
   &SkeletalMeshSkeletonLocation(Module.Inherit Velocity=false, Module.Use Sampled Bone as Ribbon ID=true)
   &CylinderLocation(
    Cylinder Height=5
    Cylinder Radius=25
    MaskSpawnGroup=false
    Module.Enable Lathe Profile=false
    Module.Override Randomness=false
    Module.Surface Only=false
    Orientation Axis:ENiagaraOrientationAxis=::YAxis
   )
   &SubUVAnimation(SubUV Animation Mode=::NewEnumerator1)
   /SetVariables(Particles.SpriteSize=&Vector2DFromFloat(Value=&UniformRangedFloat(Minimum=3, Maximum=10))) {isSetParameters=true, module=SetVariables_#4}
   &GravityForce(Gravity=&FN_FrontEndVector3Modifier(Gameplay=<0 0 -400>, Front End=<0 0 75>))
   &ApplyInitialForces()
   &AddVelocity(Velocity=&FN_FrontEndVector3Modifier(
     Gameplay=&UniformRangedVector(Minimum=<-10 150 -10>, Maximum=<10 300 10>)
     Front End=&UniformRangedVector(Minimum=<-10 -150 -10>, Maximum=<10 -300 10>)
    ))
  ]
  particleUpdateScript [
   &ParticleState()
   &SolveForcesAndVelocity(Clamp Velocity=false, Limit Acceleration=false)
   /SetVariables(Particles.Velocity=&ScaleVectorByCurve(
     InitialValue:Vector3f=&UniformRangedVector(Minimum=<-3 -3 0>, Maximum=<3 3 0>)
     ScaleCurve:NiagaraDataInterfaceVectorCurve=DataInterface() {propertyValues {
       XCurve {keys [<at 0 50 RCIM_Cubic>, {arriveTangent=-241.6722, interpMode=RCIM_Cubic, leaveTangent=-241.672, tangentMode=RCTM_User, time=0.5, value=20}]}
       YCurve {keys [<at 0 50 RCIM_Cubic>, {arriveTangent=-247.443, interpMode=RCIM_Cubic, leaveTangent=-247.448, tangentMode=RCTM_User, time=0.4991, value=20}]}
       ZCurve {keys [<at 0 0>,<at 1 0>]}
      }}
     CurveIndex:NiagaraFloat=$Particles.NormalizedAge
    )) {enabled=false, isSetParameters=true, module=SetVariables_#6}
   /SetVariables() {enabled=false, isSetParameters=true, module=SetVariables_#7}
   &ScaleSpriteSize(Scale Factor=&Vector2DFromFloat(Value=&FloatFromCurve(FloatCurve=DataInterface() {propertyValues {Curve {keys [<at 0 0 RCIM_Cubic>,<at 0.05 1 RCIM_Cubic>,<at 1 0>]}}})))
  ]
  rendererClasses [&NiagaraSpriteRendererProperties]
  renderers [{rendererClass=&NiagaraSpriteRendererProperties}]
 }
 Wisp_Ribbon {
  emitterUpdateScript [
   &EmitterState(Life Cycle Mode=::NewEnumerator1, Loop Duration=0.65, UseLoopDelay=false)
   &FN_SpawnRibbon_v2(Number of Ribbons=2, Spawn Mode=::NewEnumerator0, SpawnRate=25)
   &SpawnPerUnit(Module.Use Max Movement Threshold=true, Module.Use Movement Tolerance=false, Use Spawn Probability=false) {enabled=false}
  ]
  particleSpawnScript [
   &FN_InitializeRibbon_v3(Color:LinearColor=<rgba 0.2174 0.2713 0.33 0.5>, Enable Skeletal Position=true, Lifetime=1.5, Offset Amount=-25, Ribbon Width=75)
   &GravityForce(Gravity=&FN_FrontEndVector3Modifier(Gameplay=<0 0 -400>, Front End=<0 0 75>))
   &ApplyInitialForces()
   &AddVelocity(Velocity=&FN_FrontEndVector3Modifier(Gameplay=<0 300 0>, Front End=<0 -200 0>)) {module=AddVelocity002}
  ]
  particleUpdateScript [
   &ParticleState()
   &SolveForcesAndVelocity(Clamp Velocity=false, Limit Acceleration=false)
   /Game/Effects/Niagara/Modules/Contrails/FN_UpdateRibbon/FN_UpdateRibbon_v2()
   /SetVariables(Particles.RibbonWidth=&Multiply_Float(
     A=$Particles.Initial.RibbonWidth
     B=&FloatFromCurve(FloatCurve=DataInterface() {propertyValues {Curve {keys [<at 0 1>,<at 1 0>]}}})
    )) {enabled=false, isSetParameters=true}
   &ScaleColor(
    Scale Alpha=&FloatFromCurve(FloatCurve=DataInterface() {propertyValues {Curve {keys [
         {arriveTangent=1.8999, interpMode=RCIM_Cubic, time=-0.0044, value=0}
         {arriveTangent=11.5714, interpMode=RCIM_Cubic, leaveTangent=-4.1242, tangentMode=RCTM_Break, time=0.22, value=1}
         {arriveTangent=-0.0368, interpMode=RCIM_Cubic, leaveTangent=-0.0368, tangentMode=RCTM_User, time=0.9, value=0}
        ]}}})
    ScaleA=true
    ScaleRGB=false
   )
   &Material(Base Texture Time:NiagaraFloat=&Multiply_Float(A=$Emitter.Age, B=-1), PostMult=0.5455, Power=0.4408)
  ]
  rendererClasses [&NiagaraRibbonRendererProperties]
  renderers [{rendererClass=&NiagaraRibbonRendererProperties}]
 }
}
summary {
 systemName=NS_Glider_TacticalScuba_Male
 emitters [
  {emitterName=Prop_Bubbles, rendererClasses [&NiagaraSpriteRendererProperties]}
  {emitterName=Wisp_Ribbon, rendererClasses [&NiagaraRibbonRendererProperties]}
  {emitterName=Bubbles_Wings_Turning, rendererClasses [&NiagaraSpriteRendererProperties]}
  {emitterName=Bubbles_Wings, rendererClasses [&NiagaraSpriteRendererProperties]}
  {
   emitterName=Particle_Light
   rendererClasses [&NiagaraLightRendererProperties, &NiagaraSpriteRendererProperties]
  }
 ]
 userVariables [
  User.bIsFrontEnd {type=NiagaraBool}
  User.bIsFrontEndPreview {type=NiagaraBool}
  User.bIsFullyDeployed {type=NiagaraBool}
  User.ForwardDot {type=NiagaraFloat, defaultValue=0, description=""}
  User.LightColor {type=LinearColor, description="", defaultValue {struct=LinearColor, value=<rgba 50 0 44.176 1>}}
  User.PlayerSpeed {type=NiagaraFloat, defaultValue=0, description=""}
  User.RibbonIDOffset {type=NiagaraInt32}
  User.RightDot {type=NiagaraFloat, defaultValue=0, description=""}
  User.SkeletalMesh {
   type=NiagaraDataInterfaceSkeletalMesh
   defaultValue {
    struct=NiagaraExt_VariableValue_DataInterface
    value {dataInterface=&self:NiagaraDataInterfaceSkeletalMesh_0}
   }
  }
 ]
}
